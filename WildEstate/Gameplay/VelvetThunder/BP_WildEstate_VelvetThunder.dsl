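; Interactable Velvet Thunder slushie machine that dispenses a "Super Squishee" pickup, currently a temporary item.
; Holding interact plays a looping sound and a timeline that drives shake and fill level through custom primitive data; interrupting resets them.
; Completing it tosses the pickup with a bell and Niagara effect, then briefly blocks interaction on a timer.
; Reaching the grantable item count disables interaction, empties the fill and calls an empty ItemsDepleted event.
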(event Custom|ItemsDepleted)

(event Custom|GiveTreat
  (bind (_returnvalue_roll _returnvalue_pitch _returnvalue_yaw) (Math|Rotator|BreakRotator (Math|Rotator|MakeRotfromX (Transformation|GetActorUpVector))))
  (Utilities|FlowControl|Sequence
    (:then_0
      (Rendering|Material|SetCustomPrimitiveDataFloat (Variables|Debug|GetStaticMeshComponent) 4 0.0)
      (Audio|PlaySound2D "/Game/Sounds/Fort_UI_Sounds/Gameplay/bell_tr_01.bell_tr_01" 2.0)
      ;[0.9043 0.002344 0.002344] TEMP Item drop for Super Squishee. Should be replaced with a loot tier group (below) once we get it.
      (bind _returnvalue (Inventory|SpawnPickupInWorld (Variables|Default|GetTempItemGranted) 1 (+ (Transformation|GetWorldLocation (Variables|Debug|GetStaticMeshComponent)) (* (Transformation|GetUpVector (Variables|Debug|GetStaticMeshComponent)) 5e+01)) "-1.000000,0.000000,1.000000" 0 true true false 0 "Container" "Chest"))
      (Fortnite|Container|TossPickupfromContainer self _returnvalue 0 0 3.6e+02 (Math|Rotator|MakeRotator 0.0 35.0 _returnvalue_yaw) (Math|Random|RandomFloatinRange 6e+02 8e+02) false false 0)
      (Niagara|SpawnSystematLocation :SystemTemplate "/WildEstateTech/FX/VelvetThunder/NS_VelvetThunder_Eject_01.NS_VelvetThunder_Eject_01" :Location (Transformation|GetActorLocation self) :Scale "1.000000,1.000000,1.000000" :bAutoDestroy false :bAutoActivate true :PoolingMethod "None" :bPreCullCheck true))
    (:then_1
      (if (== (Variables|Default|GetGrantableItemCount) (Variables|Default|GetUseTimes))
        (Variables|Interaction|SetbAllowInteract false)
        (Rendering|Material|SetCustomPrimitiveDataFloat (Variables|Debug|GetStaticMeshComponent) 5 0.0)
        (CallFunction|ItemsDepleted)
        (else
          (Math|Integer|IncrementInt (Variables|Default|GetUseTimes)
            (:%20%20
              (Variables|Interaction|SetbAllowInteract false)
              (Utilities|Time|SetTimerbyEvent (delegate Custom|AllowInteract) 2.0))))))))

(event Custom|AllowInteract
  (Variables|Interaction|SetbAllowInteract true))

(event Interaction|EventBlueprintOnInteract
  (params (InteractingPawn :type Fort%20Pawn%20Object%20Reference) (InteractionBeingAttempted :type EInteractionBeingAttempted%20Enum :default "FirstInteraction"))
  (CallFunction|GiveTreat))

(event Interaction|EventBlueprintOnBeginInteract
  (params (InteractingPawn :type Fort%20Pawn%20Object%20Reference) (InteractionBeingAttempted :type EInteractionBeingAttempted%20Enum :default "FirstInteraction") (InteractionType :type TInteractionType%20Enum :default "IT_NoInteraction"))
  (bind _output_get (Variables|Default|Setinitialworldlocation (Transformation|GetWorldLocation (Variables|Debug|GetStaticMeshComponent))))
  (bind _returnvalue (Audio|SpawnSoundatLocation "/FNGameplayCues/Sounds/Fort_GamePlay_Sounds/BRSword/AB_Pull_Stress_Loop_cue.AB_Pull_Stress_Loop_cue" _output_get))
  (Variables|Default|SetActiveAudioLoop _returnvalue)
  (Timeline|VelvetMotion PlayFromStart
    (:Update
      (bind _output_get_2 (Variables|Default|SetShakeIntensity (Math|Float|Clamp(Float) (+ (Variables|Default|GetShakeIntensity) 0.008))))
      (CallFunction|SetShakeIntensityVisuals _output_get_2)
      (bind _output_get_3 (Variables|Default|SetSlushAmount (Math|Float|Clamp(Float) (- (Variables|Default|GetSlushAmount) 0.002))))
      (Development|PrintString (Utilities|String|ToString(Float) _output_get_3) true true "(R=0.000000,G=0.660000,B=1.000000,A=1.000000)" _output_get_3)
      (CallFunction|SetFillLevelVisuals (Variables|Default|GetSlushAmount)))
    (:Finished
      (Audio|Components|Audio|Stop (Variables|Default|GetActiveAudioLoop)))))

(event Interaction|EventBlueprintOnInterruptInteract
  (Audio|Components|Audio|Stop (Variables|Default|GetActiveAudioLoop))
  (Components|Timeline|Stop (Variables|BP_WildEstate_VelvetThunder|GetVelvetMotion))
  (Rendering|Material|SetCustomPrimitiveDataFloat (Variables|Debug|GetStaticMeshComponent) 4 0.0)
  (bind _output_get (Variables|Default|SetSlushAmount 0.9))
  (Rendering|Material|SetCustomPrimitiveDataFloat (Variables|Debug|GetStaticMeshComponent) 5 _output_get))

(fn BlueprintCanInteract
  (params (InteractingPawn :type Fort%20Pawn%20Object%20Reference) (InteractionBeingAttempted :type EInteractionBeingAttempted%20Enum :default "FirstInteraction") (InteractionType :type TInteractionType%20Enum :default "IT_NoInteraction"))
  (return (Variables|Interaction|GetbAllowInteract)))

(fn BlueprintGetInteractionString
  (params (InteractingPawn :type Fort%20Pawn%20Object%20Reference) (InteractionBeingAttempted :type EInteractionBeingAttempted%20Enum :default "FirstInteraction"))
  (return "NSLOCTEXT(\"[9928BFAEB0E7D60589F8C699D832724E]\", \"5AEAAB344D74CD550E873BBD572E6B14\", \"Make Super Squishee\")"))

(fn BlueprintOnLocalInteract
  (params (InteractingPawn :type Fort%20Player%20Pawn%20Object%20Reference) (InteractionBeingAttempted :type EInteractionBeingAttempted%20Enum :default "FirstInteraction"))
  (CallFunction|GiveTreat)
  (return false))

(fn SetFillLevelVisuals
  (params (Alpha :type Float%20%28single-precision%29))
  (Rendering|Material|SetCustomPrimitiveDataFloat (Variables|Default|GetSM_FrozenFood_SlushieLargeCupGlass_A) 5 Alpha))

(fn SetShakeIntensityVisuals
  (params (Alpha :type Float%20%28single-precision%29))
  (Rendering|Material|SetCustomPrimitiveDataFloat (Variables|Default|GetSM_FrozenFood_SlushieLargeCupGlass_A) 4 Alpha))

(fn ConstructionScript)
